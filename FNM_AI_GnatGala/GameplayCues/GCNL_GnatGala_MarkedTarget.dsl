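; Looping gameplay cue for GnatGala's marked target that shows a Niagara camera lens effect to the target's player.
; On start it adds the lens effect via the player camera manager, caches its Niagara component and sets FadeInOut true.
; A looping 0.1-second timer feeds the viewport aspect ratio into the effect's Ratio parameter.
; On removal it fades out, clears the timer, then deactivates the component after a one-second delay.
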(event EventOnLoopingStart(EitherVFX-preferthisversion)
  (params (MyTarget :type Actor%20Object%20Reference) (Parameters :type Gameplay%20Cue%20Parameters%20Structure%20%28by%20ref%29) (ParticleComponents :type Array%20of%20FXSystem%20Component%20Object%20References) (AudioComponents :type Array%20of%20Audio%20Component%20Object%20References))
  (bind (_x _y) (Math|Vector2D|BreakVector2D (Viewport|GetViewportSize (Variables|Default|GetMy%20Target))))
  (|CallParentFunction MyTarget Parameters ParticleComponents AudioComponents)
  (Variables|Default|SetMy%20Target MyTarget)
  (bind _returnvalue (Camera|AddGenericCameraLensEffect (Variables|/Script/Engine.PlayerController|GetPlayerCameraManager (Fortnite|GetFortPlayerControllerfromActor MyTarget)) (Variables|Default|GetLens%20Effect)))
  (bind _output_get (Variables|Default|SetCamera%20Lens%20Effect _returnvalue))
  (bind _asniagara_particle_system_component (Utilities|Casting|CastToNiagaraComponent (CameraLensEffect|GetPrimaryParticleComponent _output_get)))
  (Variables|Default|SetNS%20Camera%20FX _asniagara_particle_system_component)
  (bind _returnvalue_2 (Utilities|Time|SetTimerbyEvent (EventDispatchers|CreateEvent :!event Custom|Camera%20Changed%20Checker) 0.1 true true))
  (Variables|Default|SetTimer _returnvalue_2)
  (Niagara|SetNiagaraVariable(Bool) (Variables|Default|GetNS%20Camera%20FX) "FadeInOut" true)
  (Class|BuildingFunctionLibrary|LogString "NewEnumerator3" (Utilities|String|Append "Viewport Size X: " (Utilities|String|ToString(Float) _x) "
" "Viewport Size Y:" (Utilities|String|ToString(Float) _y))))

(event Custom|Camera%20Changed%20Checker
  (bind (_x _y) (Math|Vector2D|BreakVector2D (Viewport|GetViewportSize (Variables|Default|GetMy%20Target))))
  (Niagara|SetNiagaraVariable(Vector2) (Variables|Default|GetNS%20Camera%20FX) "Ratio" (Math|Vector2D|MakeVector2D (/ _x _y) 1.0)))

(event EventOnRemoval(EitherVFX-preferthisversion)
  (params (MyTarget :type Actor%20Object%20Reference) (Parameters :type Gameplay%20Cue%20Parameters%20Structure%20%28by%20ref%29) (ParticleComponents :type Array%20of%20FXSystem%20Component%20Object%20References) (AudioComponents :type Array%20of%20Audio%20Component%20Object%20References) (BurstCameraShakeInstance :type Camera%20Shake%20Base%20Object%20Reference) (BurstDecalInstance :type Decal%20Actor%20Object%20Reference))
  (|CallParentFunction MyTarget Parameters ParticleComponents AudioComponents BurstCameraShakeInstance BurstDecalInstance)
  (Niagara|SetNiagaraVariable(Bool) (Variables|Default|GetNS%20Camera%20FX) "FadeInOut" false)
  (Utilities|Time|ClearandInvalidateTimerbyHandle 0 (Variables|Default|GetTimer))
  (Utilities|FlowControl|Delay 0 1.0)
  (Components|Activation|Deactivate (Variables|Default|GetNS%20Camera%20FX)))

(fn ConstructionScript)
